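# Physics asset for the DapperPunch medium male player body, covering the torso, limbs and dynamic clothing and accessory bones.
# Body capsules use free angular constraints, and a huge root sphere sits below the character, its top near foot level like a floor.
# Two tuxedo tail chains hang from a shared base on spine_01 with tight limits.
# Shirt chains and the right-thigh holster swing on restricted axes, while four bag bodies are angularly locked to spine_04.
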PhysicsAsset /Game/Characters/Player/Male/Medium/Bodies/M_MED_DapperPunch/Meshes/M_MED_DapperPunch_Physics
bodies {
 calf_l [{shapeType=Capsule, length=11.2207, radius=5, shapeName=calf_l_capsule, center=<-9.6529 -1.5602 0>, rotation=<rot 90 83.2785 80.2759>}]
 calf_r [{shapeType=Capsule, length=10, radius=5, shapeName=calf_r_capsule, center=<9.6529 1.5602 0>, rotation=<rot 90 83.2785 80.2759>}]
 dyn_bag_b_1_l [{shapeType=Capsule, length=1, radius=6, shapeName=dyn_bag_b_1_r_capsule, center=<2.6588 5.2309 0>, rotation=<rot 0 0 -90>}]
 dyn_bag_b_1_r [{shapeType=Capsule, length=1, radius=6, shapeName=dyn_bag_b_1_r_capsule, center=<-2.6588 -5.2309 0>, rotation=<rot 0 0 90>}]
 dyn_bag_f_1_l [{shapeType=Capsule, length=1, radius=6, shapeName=dyn_bag_f_1_r_capsule, center=<2.6588 5.2309 0>, rotation=<rot 0 0 -90>}]
 dyn_bag_f_1_r [{shapeType=Capsule, length=1, radius=6, shapeName=dyn_bag_f_1_r_capsule, center=<-2.6588 -5.2309 0>, rotation=<rot 0 0 90>}]
 dyn_holster [{shapeType=Capsule, length=4, radius=5, shapeName=dyn_holster_capsule, center=<0 -0.2793 0>, rotation=<rot 90 0 0>}]
 dyn_shirt_1_l [{shapeType=Capsule, length=4.2207, radius=2.0839, shapeName=dyn_shirt_1_l_capsule, center=<0 -2.0607 0>}]
 dyn_shirt_1_r [{shapeType=Capsule, length=4.2207, radius=2.0839, shapeName=dyn_shirt_1_l_capsule, center=<0 2.0607 0>}]
 dyn_shirt_2_l [{shapeType=Capsule, length=7.7692, radius=1.5316, shapeName=dyn_shirt_2_l_capsule, center=<4.1461 -2.0607 0>}]
 dyn_shirt_2_r [{shapeType=Capsule, length=8.4684, radius=1.5316, shapeName=dyn_shirt_2_l_capsule, center=<-4.1461 2.0607 0>}]
 dyn_tuxedo_b_1_l [{shapeType=Capsule, length=7.1585, radius=4.29, shapeName=dyn_tuxedo_b_1_r_capsule, center=<0.3382 -3.7839 0>, rotation=<rot 0 0 -180>}]
 dyn_tuxedo_b_1_r [{shapeType=Capsule, length=7.1585, radius=4.29, shapeName=dyn_tuxedo_b_1_r_capsule, center=<-0.3382 3.7839 0>}]
 dyn_tuxedo_b_2_l [{shapeType=Capsule, length=3.5243, radius=5, shapeName=dyn_tuxedo_b_2_r_capsule, center=<0.3382 -3.7847 0>, rotation=<rot 0 0 -180>}]
 dyn_tuxedo_b_2_r [{shapeType=Capsule, length=3.5243, radius=5, shapeName=dyn_tuxedo_b_2_r_capsule, center=<-0.3382 3.7847 0>}]
 dyn_tuxedo_b_3_l [{shapeType=Capsule, length=8, radius=5, shapeName=dyn_tuxedo_b_3_r_capsule, center=<1.1762 -3.7847 0>, rotation=<rot 90 0 180>}]
 dyn_tuxedo_b_3_r [{shapeType=Capsule, length=8, radius=5, shapeName=dyn_tuxedo_b_3_r_capsule, center=<-1.1762 3.7847 0>, rotation=<rot 90 0 0>}]
 dyn_tuxedo_b_base [{shapeType=Capsule, length=8, radius=5, shapeName=dyn_tuxedo_b_base_capsule, center=<0.3382 0 0>}]
 pelvis [{shapeType=Capsule, length=9.6085, radius=10.3915, shapeName=pelvis_capsule, center=<0 0.7743 0>}]
 root [{shapeType=Sphere, radius=465.1064, shapeName=root_sphere, center=<0 -36.0449 -461.0143>}]
 spine_01 [{shapeType=Capsule, length=10, radius=8, shapeName=spine_01_capsule}]
 spine_02 [{shapeType=Capsule, length=10, radius=8, shapeName=spine_02_capsule}]
 spine_03 [{shapeType=Capsule, length=10, radius=8, shapeName=spine_03_capsule}]
 spine_04 [{shapeType=Capsule, length=10, radius=10, shapeName=spine_04_capsule}]
 spine_05 [{shapeType=Capsule, length=10, radius=10, shapeName=spine_05_capsule}]
 thigh_l [{shapeType=Capsule, length=32.3665, radius=7.9784, shapeName=thigh_l_capsule, center=<-14.7821 -0.8051 0>, rotation=<rot 90 0 0>}]
 thigh_r [{shapeType=Capsule, length=32.3665, radius=7.9784, shapeName=thigh_l_capsule, center=<14.7821 0.8051 0>, rotation=<rot 90 0 180>}]
 upperarm_l [{shapeType=Capsule, length=8, radius=6, shapeName=upperarm_r_capsule, center=<9.9826 0 0>, rotation=<rot 90 0 180>}]
 upperarm_r [{shapeType=Capsule, length=8, radius=6, shapeName=upperarm_r_capsule, center=<-9.9826 0 0>, rotation=<rot 90 0 0>}]
}
constraints [
 {
  bone1Name=spine_01
  bone2Name=pelvis
  swing1LimitDegrees=45
  swing1Motion=Free
  swing2LimitDegrees=45
  swing2Motion=Free
  twistLimitDegrees=45
  twistMotion=Free
 }
 {
  bone1Name=spine_02
  bone2Name=spine_01
  swing1LimitDegrees=45
  swing1Motion=Free
  swing2LimitDegrees=45
  swing2Motion=Free
  twistLimitDegrees=45
  twistMotion=Free
 }
 {
  bone1Name=spine_03
  bone2Name=spine_02
  swing1LimitDegrees=45
  swing1Motion=Free
  swing2LimitDegrees=45
  swing2Motion=Free
  twistLimitDegrees=45
  twistMotion=Free
 }
 {
  bone1Name=spine_04
  bone2Name=spine_03
  swing1LimitDegrees=45
  swing1Motion=Free
  swing2LimitDegrees=45
  swing2Motion=Free
  twistLimitDegrees=45
  twistMotion=Free
 }
 {
  bone1Name=spine_05
  bone2Name=spine_04
  swing1LimitDegrees=45
  swing1Motion=Free
  swing2LimitDegrees=45
  swing2Motion=Free
  twistLimitDegrees=45
  twistMotion=Free
 }
 {bone1Name=dyn_tuxedo_b_3_r, bone2Name=dyn_tuxedo_b_2_r, swing1LimitDegrees=30, swing2LimitDegrees=10, twistLimitDegrees=5}
 {bone1Name=dyn_tuxedo_b_base, bone2Name=spine_01, swing1LimitDegrees=5, swing2LimitDegrees=5, twistLimitDegrees=0}
 {bone1Name=dyn_tuxedo_b_1_r, bone2Name=dyn_tuxedo_b_base, swing1LimitDegrees=20, swing2LimitDegrees=15, twistLimitDegrees=5}
 {bone1Name=dyn_tuxedo_b_2_r, bone2Name=dyn_tuxedo_b_1_r, swing1LimitDegrees=30, swing2LimitDegrees=10, twistLimitDegrees=5}
 {
  bone1Name=calf_l
  bone2Name=thigh_l
  swing1LimitDegrees=45
  swing1Motion=Free
  swing2LimitDegrees=45
  swing2Motion=Free
  twistLimitDegrees=45
  twistMotion=Free
 }
 {
  bone1Name=calf_r
  bone2Name=thigh_r
  swing1LimitDegrees=45
  swing1Motion=Free
  swing2LimitDegrees=45
  swing2Motion=Free
  twistLimitDegrees=45
  twistMotion=Free
 }
 {
  bone1Name=thigh_l
  bone2Name=pelvis
  swing1LimitDegrees=45
  swing1Motion=Free
  swing2LimitDegrees=45
  swing2Motion=Free
  twistLimitDegrees=45
  twistMotion=Free
 }
 {
  bone1Name=dyn_shirt_1_l
  bone2Name=spine_01
  swing1LimitDegrees=80
  swing1Motion=Locked
  swing2LimitDegrees=0
  swing2Motion=Locked
  twistLimitDegrees=0
  twistMotion=Locked
 }
 {
  bone1Name=dyn_shirt_2_l
  bone2Name=dyn_shirt_1_l
  swing1LimitDegrees=15
  swing1Motion=Locked
  swing2LimitDegrees=15
  twistLimitDegrees=45
  twistMotion=Locked
 }
 {
  bone1Name=dyn_shirt_1_r
  bone2Name=spine_01
  swing1LimitDegrees=80
  swing1Motion=Locked
  swing2LimitDegrees=0
  swing2Motion=Locked
  twistLimitDegrees=0
  twistMotion=Locked
 }
 {
  bone1Name=dyn_shirt_2_r
  bone2Name=dyn_shirt_1_r
  swing1LimitDegrees=15
  swing1Motion=Locked
  swing2LimitDegrees=15
  twistLimitDegrees=45
  twistMotion=Locked
 }
 {
  bone1Name=thigh_r
  bone2Name=pelvis
  swing1LimitDegrees=45
  swing1Motion=Free
  swing2LimitDegrees=45
  swing2Motion=Free
  twistLimitDegrees=45
  twistMotion=Free
 }
 {
  bone1Name=pelvis
  bone2Name=root
  swing1LimitDegrees=45
  swing1Motion=Free
  swing2LimitDegrees=45
  swing2Motion=Free
  twistLimitDegrees=45
  twistMotion=Free
 }
 {bone1Name=dyn_tuxedo_b_1_l, bone2Name=dyn_tuxedo_b_base, swing1LimitDegrees=20, swing2LimitDegrees=15, twistLimitDegrees=5}
 {bone1Name=dyn_tuxedo_b_2_l, bone2Name=dyn_tuxedo_b_1_l, swing1LimitDegrees=30, swing2LimitDegrees=10, twistLimitDegrees=5}
 {bone1Name=dyn_tuxedo_b_3_l, bone2Name=dyn_tuxedo_b_2_l, swing1LimitDegrees=30, swing2LimitDegrees=10, twistLimitDegrees=5}
 {bone1Name=dyn_holster, bone2Name=thigh_r, swing1LimitDegrees=45, swing1Motion=Locked, swing2LimitDegrees=3, twistLimitDegrees=3}
 {
  bone1Name=dyn_bag_b_1_r
  bone2Name=spine_04
  swing1LimitDegrees=10
  swing1Motion=Locked
  swing2LimitDegrees=10
  swing2Motion=Locked
  twistLimitDegrees=10
  twistMotion=Locked
 }
 {
  bone1Name=dyn_bag_f_1_r
  bone2Name=spine_04
  swing1LimitDegrees=10
  swing1Motion=Locked
  swing2LimitDegrees=10
  swing2Motion=Locked
  twistLimitDegrees=10
  twistMotion=Locked
 }
 {
  bone1Name=upperarm_r
  bone2Name=spine_05
  swing1LimitDegrees=45
  swing1Motion=Free
  swing2LimitDegrees=45
  swing2Motion=Free
  twistLimitDegrees=45
  twistMotion=Free
 }
 {
  bone1Name=dyn_bag_b_1_l
  bone2Name=spine_04
  swing1LimitDegrees=10
  swing1Motion=Locked
  swing2LimitDegrees=10
  swing2Motion=Locked
  twistLimitDegrees=10
  twistMotion=Locked
 }
 {
  bone1Name=dyn_bag_f_1_l
  bone2Name=spine_04
  swing1LimitDegrees=10
  swing1Motion=Locked
  swing2LimitDegrees=10
  swing2Motion=Locked
  twistLimitDegrees=10
  twistMotion=Locked
 }
 {
  bone1Name=upperarm_l
  bone2Name=spine_05
  swing1LimitDegrees=45
  swing1Motion=Free
  swing2LimitDegrees=45
  swing2Motion=Free
  twistLimitDegrees=45
  twistMotion=Free
 }
]
