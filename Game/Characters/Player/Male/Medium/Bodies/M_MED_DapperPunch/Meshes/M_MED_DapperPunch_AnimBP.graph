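# Animation Blueprint for the DapperPunch player body, adding secondary motion to tuxedo, shirt, pants and bag bones.
# The pose copied from the parent mesh runs through RigidBody and per-bone pants AnimDynamics under overridable gravity.
# Additive bone offsets, around a Control Rig pass, then correct the result for movement states and upperarm angles.
# The EventGraph computes gravity, maps arm angles to bag alphas and smooths crouch, slide and movement alphas.
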AnimBlueprint /Game/Characters/Player/Male/Medium/Bodies/M_MED_DapperPunch/Meshes/M_MED_DapperPunch_AnimBP
target_skeleton=&/M_MED_DapperPunch_Skeleton
graphs [
 AnimGraph {nodes [
   n1 = "Transform (Modify) Bone - Bone: dyn_tuxedo_b_base"!(
    ComponentPose=Local To Component(LocalPose=Use cached pose 'SavedPose28735'())
    bAlphaBoolEnabled:bool=Get bIsSkyDiving()
    Translation="-0.144,-0.461,0.000"
    Rotation="-0.000,-17.271,0.000"
    Scale="1,1,1"
   ) -> (Pose) {
    alphaInputType=Bool
    rotationMode=BMM_Additive
    rotationSpace=BCS_BoneSpace
    translationMode=BMM_Additive
    translationSpace=BCS_BoneSpace
    alphaBoolBlend {blendInTime=0.1, blendOption=Linear, blendOutTime=0.1}
    boneToModify {boneName=dyn_tuxedo_b_base}
    rotation=<rot 0 -17.271 0>
    translation=<-0.144 -0.461>
   }
   n2 = "Transform (Modify) Bone - Bone: dyn_tuxedo_b_base"(ComponentPose=@n1, bAlphaBoolEnabled=Get bIsKneeSliding(), Translation="-0.655,-6.201,0.051", Rotation="5.187,-45,18.716", Scale="1,1,1") {
    alphaInputType=Bool
    rotationMode=BMM_Additive
    rotationSpace=BCS_BoneSpace
    translationMode=BMM_Additive
    translationSpace=BCS_BoneSpace
    alphaBoolBlend {blendInTime=0.1, blendOption=Linear, blendOutTime=0.1}
    boneToModify {boneName=dyn_tuxedo_b_base}
    rotation=<rot 5.187 -45 18.716>
    translation=<-0.655 -6.201 0.051>
   }
   n3 = Get bIsBaseballSliding()
   n4 = "Transform (Modify) Bone - Bone: dyn_tuxedo_b_base"(ComponentPose=@n2, bAlphaBoolEnabled=@n3, Translation="-1.238,-1.645,0.993", Rotation="-1.336,-82.795,45.758", Scale="1,1,1") {
    alphaInputType=Bool
    rotationMode=BMM_Additive
    rotationSpace=BCS_BoneSpace
    translationMode=BMM_Additive
    translationSpace=BCS_BoneSpace
    alphaBoolBlend {blendInTime=0.5, blendOption=Linear, blendOutTime=0.1}
    boneToModify {boneName=dyn_tuxedo_b_base}
    rotation=<rot -1.336 -82.795 45.758>
    translation=<-1.238 -1.645 0.993>
   }
   n5 = "Transform (Modify) Bone - Bone: dyn_tuxedo_b_1_l"!(ComponentPose=@n4, bAlphaBoolEnabled:bool=@n3, Translation="0,0,0", Rotation="0.021,-26.378,-0.000", Scale="1,1,1") -> (Pose) {
    alphaInputType=Bool
    rotationMode=BMM_Additive
    rotationSpace=BCS_BoneSpace
    translationMode=BMM_Additive
    translationSpace=BCS_BoneSpace
    alphaBoolBlend {blendInTime=0.5, blendOption=Linear, blendOutTime=0.1}
    boneToModify {boneName=dyn_tuxedo_b_1_l}
    rotation=<rot 0.021 -26.378 0>
   }
   n6 = "Transform (Modify) Bone - Bone: dyn_tuxedo_b_1_r"!(ComponentPose=@n5, bAlphaBoolEnabled:bool=@n3, Translation="0,0,0", Rotation="0.000,-38.027,0.000", Scale="1,1,1") -> (Pose) {
    alphaInputType=Bool
    rotationMode=BMM_Additive
    rotationSpace=BCS_BoneSpace
    translationMode=BMM_Additive
    translationSpace=BCS_BoneSpace
    alphaBoolBlend {blendInTime=0.5, blendOption=Linear, blendOutTime=0.1}
    boneToModify {boneName=dyn_tuxedo_b_1_r}
    rotation=<rot 0 -38.027 0>
   }
   n7 = "Transform (Modify) Bone - Bone: dyn_pants_inn_r"!(ComponentPose=@n6, bAlphaBoolEnabled:bool=True, Translation="0,0,0", Rotation="-0.016,-4.279,6.098", Scale="1,1,1") -> (Pose) {
    alphaInputType=Bool
    rotationMode=BMM_Additive
    rotationSpace=BCS_BoneSpace
    translationMode=BMM_Additive
    translationSpace=BCS_BoneSpace
    alphaBoolBlend {blendInTime=0.1, blendOption=Linear, blendOutTime=0.1}
    boneToModify {boneName=dyn_pants_inn_r}
    rotation=<rot -0.016 -4.279 6.098>
   }
   n8 = "Transform (Modify) Bone - Bone: dyn_bag_b_r"(
    ComponentPose=@n7
    Alpha=Get r_upperarm_angle_for_b_bag() -> (r_upperarm_angle_for_b_bag:double)
    Translation="0,0,0"
    Rotation="0,0,0"
    Scale="1,1,1"
   ) {
    rotationMode=BMM_Additive
    rotationSpace=BCS_BoneSpace
    scaleMode=BMM_Additive
    scaleSpace=BCS_BoneSpace
    translationMode=BMM_Additive
    translationSpace=BCS_BoneSpace
    alphaBoolBlend {blendInTime=0.1, blendOption=Linear, blendOutTime=0.1}
    alphaScaleBiasClamp {inRange {max=30, min=70}}
    boneToModify {boneName=dyn_bag_b_r}
   }
   n9 = "Transform (Modify) Bone - Bone: dyn_bag_f_r"(
    ComponentPose=@n8
    Alpha=Get r_upperarm_angle_for_f_bag() -> (r_upperarm_angle_for_f_bag:double)
    Translation="0.000,0.000,0.432"
    Rotation="14.228,0.000,0.000"
    Scale="1.000,1.000,0.616"
   ) {
    rotationMode=BMM_Additive
    rotationSpace=BCS_BoneSpace
    scaleMode=BMM_Additive
    scaleSpace=BCS_BoneSpace
    translationMode=BMM_Additive
    translationSpace=BCS_BoneSpace
    alphaBoolBlend {blendInTime=0.1, blendOption=Linear, blendOutTime=0.1}
    alphaScaleBiasClamp {inRange {max=30, min=70}}
    boneToModify {boneName=dyn_bag_f_r}
    rotation=<rot 14.228 0 0>
    scale {z=0.616}
    translation {z=0.432}
   }
   n10 = "Transform (Modify) Bone - Bone: dyn_bag_b_l"(
    ComponentPose=@n9
    Alpha=Get l_upperarm_angle_for_b_bag() -> (l_upperarm_angle_for_b_bag:double)
    Translation="-0.000,0.063,0.988"
    Rotation="0.000,0.001,3.688"
    Scale="1.000,1.000,0.528"
   ) {
    rotationMode=BMM_Additive
    rotationSpace=BCS_BoneSpace
    scaleMode=BMM_Additive
    scaleSpace=BCS_BoneSpace
    translationMode=BMM_Additive
    translationSpace=BCS_BoneSpace
    alphaBoolBlend {blendInTime=0.1, blendOption=Linear, blendOutTime=0.1}
    alphaScaleBiasClamp {inRange {max=30, min=70}}
    boneToModify {boneName=dyn_bag_b_l}
    rotation=<rot 0 0.001 3.688>
    scale {z=0.528}
    translation {y=0.063, z=0.988}
   }
   n11 = "Transform (Modify) Bone - Bone: dyn_bag_f_l"(
    ComponentPose=@n10
    Alpha=Get l_upperarm_angle_for_f_bag() -> (l_upperarm_angle_for_f_bag:double)
    Translation="-0.000,0.113,-2.690"
    Rotation="0.000,-0.000,-2.009"
    Scale="1.000,1.000,0.468"
   ) {
    rotationMode=BMM_Additive
    rotationSpace=BCS_BoneSpace
    scaleMode=BMM_Additive
    scaleSpace=BCS_BoneSpace
    translationMode=BMM_Additive
    translationSpace=BCS_BoneSpace
    alphaBoolBlend {blendInTime=0.1, blendOption=Linear, blendOutTime=0.1}
    alphaScaleBiasClamp {inRange {max=30, min=70}}
    boneToModify {boneName=dyn_bag_f_l}
    rotation=<rot 0 0 -2.009>
    scale {z=0.468}
    translation {y=0.113, z=-2.69}
   }
   n12 = Control Rig!(
    Source=Component To Local(ComponentPose=@n11)
    Alpha
    CrouchAlpha:double=Get CrouchingAlpha() -> (CrouchingAlpha:double)
    Sliding:double=Get Sliding() -> (Sliding:double)
    moving 2d:double=Get moving2d() -> (moving2d:double)
   ) -> (Pose) {
    alphaBoolBlend {blendOption=Linear}
    controlRigAssetReference {blueprintRigClass=&/M_MED_DapperPunch_CtrlRig.M_MED_DapperPunch_CtrlRig_C}
   }
   n13 = Get bIsMoving2D()
   n14 = "Transform (Modify) Bone - Bone: dyn_shirt_1_l"!(ComponentPose=Local To Component(LocalPose=@n12), bAlphaBoolEnabled:bool=@n13, Translation="0,-1,0", Rotation="0,0,0", Scale="1,1,1") -> (Pose) {
    alphaInputType=Bool
    rotationMode=BMM_Additive
    rotationSpace=BCS_BoneSpace
    translationMode=BMM_Additive
    translationSpace=BCS_BoneSpace
    alphaBoolBlend {blendInTime=0.2, blendOption=Linear, blendOutTime=0.2}
    alphaScaleBiasClamp {bInterpResult=true, bMapRange=true, inRange {max=-1, min=-22}}
    boneToModify {boneName=dyn_shirt_1_l}
    translation {y=-1}
   }
   n15 = "Transform (Modify) Bone - Bone: dyn_shirt_1_r"!(ComponentPose=@n14, bAlphaBoolEnabled:bool=@n13, Translation="0,1,0", Rotation="0,0,0", Scale="1,1,1") -> (Pose) {
    alphaInputType=Bool
    rotationMode=BMM_Additive
    rotationSpace=BCS_BoneSpace
    translationMode=BMM_Additive
    translationSpace=BCS_BoneSpace
    alphaBoolBlend {blendInTime=0.2, blendOption=Linear, blendOutTime=0.2}
    alphaScaleBiasClamp {bInterpResult=true, bMapRange=true, inRange {max=-1, min=-22}}
    boneToModify {boneName=dyn_shirt_1_r}
    translation {y=1}
   }
   n16 = Get bIsKneeSliding()
   n17 = "Transform (Modify) Bone - Bone: dyn_shirt_1_l"(ComponentPose=@n15, bAlphaBoolEnabled=@n16, Translation="0,-2,0", Rotation="0.000,8.996,0.000", Scale="1,1,1") {
    alphaInputType=Bool
    rotationMode=BMM_Additive
    rotationSpace=BCS_BoneSpace
    translationMode=BMM_Additive
    translationSpace=BCS_BoneSpace
    alphaBoolBlend {blendInTime=0.2, blendOption=Linear, blendOutTime=0.2}
    alphaScaleBiasClamp {bInterpResult=true, bMapRange=true, inRange {max=-1, min=-22}}
    boneToModify {boneName=dyn_shirt_1_l}
    rotation {pitch=0, yaw=8.996}
    translation {y=-2}
   }
   n18 = "Transform (Modify) Bone - Bone: dyn_shirt_1_r"(ComponentPose=@n17, bAlphaBoolEnabled=@n16, Translation="0,2,0", Rotation="0.357,9.034,-2.262", Scale="1,1,1") {
    alphaInputType=Bool
    rotationMode=BMM_Additive
    rotationSpace=BCS_BoneSpace
    translationMode=BMM_Additive
    translationSpace=BCS_BoneSpace
    alphaBoolBlend {blendInTime=0.2, blendOption=Linear, blendOutTime=0.2}
    alphaScaleBiasClamp {bInterpResult=true, bMapRange=true, inRange {max=-1, min=-22}}
    boneToModify {boneName=dyn_shirt_1_r}
    rotation=<rot 0.357 9.034 -2.262>
    translation {y=2}
   }
   n19 = Get bIsBaseballSliding()
   n20 = "Transform (Modify) Bone - Bone: dyn_shirt_1_l"(ComponentPose=@n18, bAlphaBoolEnabled=@n19, Translation="-0.035,-1.466,0.011", Rotation="-4.795,-0.000,0.000", Scale="1,1,1") {
    alphaInputType=Bool
    rotationMode=BMM_Additive
    rotationSpace=BCS_BoneSpace
    translationMode=BMM_Additive
    translationSpace=BCS_BoneSpace
    alphaBoolBlend {blendInTime=0.1, blendOption=Linear, blendOutTime=0.2}
    alphaScaleBiasClamp {bInterpResult=true, bMapRange=true, inRange {max=-1, min=-22}}
    boneToModify {boneName=dyn_shirt_1_l}
    rotation {pitch=-4.795, yaw=0}
    translation=<-0.035 -1.466 0.011>
   }
   n21 = "Transform (Modify) Bone - Bone: dyn_shirt_1_r"(ComponentPose=@n20, bAlphaBoolEnabled=@n19, Translation="0.883,-0.517,0.159", Rotation="-14.313,36.811,30.760", Scale="1,1,1") {
    alphaInputType=Bool
    rotationMode=BMM_Additive
    rotationSpace=BCS_BoneSpace
    translationMode=BMM_Additive
    translationSpace=BCS_BoneSpace
    alphaBoolBlend {blendInTime=0.1, blendOption=Linear, blendOutTime=0.2}
    alphaScaleBiasClamp {bInterpResult=true, bMapRange=true, inRange {max=-1, min=-22}}
    boneToModify {boneName=dyn_shirt_1_r}
    rotation=<rot -14.313 36.811 30.76>
    translation=<0.883 -0.517 0.159>
   }
   n22 = Get bIsSkyDiving()
   n23 = "Transform (Modify) Bone - Bone: dyn_pants_base_l"!(ComponentPose=@n21, bAlphaBoolEnabled:bool=@n22, Translation="0,0,0", Rotation="0.001,-13.976,-0.000", Scale="1,1,1") -> (Pose) {
    alphaInputType=Bool
    rotationMode=BMM_Additive
    rotationSpace=BCS_BoneSpace
    translationMode=BMM_Additive
    translationSpace=BCS_BoneSpace
    alphaBoolBlend {blendInTime=0.1, blendOption=Linear, blendOutTime=0.2}
    alphaScaleBiasClamp {bInterpResult=true, bMapRange=true, inRange {max=-1, min=-22}}
    boneToModify {boneName=dyn_pants_base_l}
    rotation=<rot 0.001 -13.976 0>
   }
   n24 = "Transform (Modify) Bone - Bone: dyn_pants_base_r"!(ComponentPose=@n23, bAlphaBoolEnabled:bool=@n22, Translation="0,0,0", Rotation="0.000,-10.475,0.000", Scale="1,1,1") -> (Pose) {
    alphaInputType=Bool
    rotationMode=BMM_Additive
    rotationSpace=BCS_BoneSpace
    translationMode=BMM_Additive
    translationSpace=BCS_BoneSpace
    alphaBoolBlend {blendInTime=0.1, blendOption=Linear, blendOutTime=0.2}
    alphaScaleBiasClamp {bInterpResult=true, bMapRange=true, inRange {max=-1, min=-22}}
    boneToModify {boneName=dyn_pants_base_r}
    rotation=<rot 0 -10.475 0>
   }
   n25 = "Transform (Modify) Bone - Bone: dyn_pants_f_l"!(ComponentPose=@n24, bAlphaBoolEnabled:bool=@n22, Translation="0,0,0", Rotation="-0.002,-10.173,0.012", Scale="1,1,1") -> (Pose) {
    alphaInputType=Bool
    rotationMode=BMM_Additive
    rotationSpace=BCS_BoneSpace
    translationMode=BMM_Additive
    translationSpace=BCS_BoneSpace
    alphaBoolBlend {blendInTime=0.1, blendOption=Linear, blendOutTime=0.2}
    alphaScaleBiasClamp {bInterpResult=true, bMapRange=true, inRange {max=-1, min=-22}}
    boneToModify {boneName=dyn_pants_f_l}
    rotation=<rot -0.002 -10.173 0.012>
   }
   n26 = "Transform (Modify) Bone - Bone: dyn_pants_inn_l"!(ComponentPose=@n25, bAlphaBoolEnabled:bool=Get bIsParachuteOpened(), Translation="0,0,0", Rotation="0.003,-10.879,-0.018", Scale="1,1,1") -> (Pose) {
    alphaInputType=Bool
    rotationMode=BMM_Additive
    rotationSpace=BCS_BoneSpace
    translationMode=BMM_Additive
    translationSpace=BCS_BoneSpace
    alphaBoolBlend {blendInTime=0.1, blendOption=Linear, blendOutTime=0.2}
    alphaScaleBiasClamp {bInterpResult=true, bMapRange=true, inRange {max=-1, min=-22}}
    boneToModify {boneName=dyn_pants_inn_l}
    rotation=<rot 0.003 -10.879 -0.018>
   }
   n27 = "Transform (Modify) Bone - Bone: dyn_pants_out_l"!(ComponentPose=@n26, bAlphaBoolEnabled:bool=@n22, Translation="0,0,0", Rotation="8.682,-7.926,0.058", Scale="1,1,1") -> (Pose) {
    alphaInputType=Bool
    rotationMode=BMM_Additive
    rotationSpace=BCS_BoneSpace
    translationMode=BMM_Additive
    translationSpace=BCS_BoneSpace
    alphaBoolBlend {blendInTime=0.1, blendOption=Linear, blendOutTime=0.2}
    alphaScaleBiasClamp {bInterpResult=true, bMapRange=true, inRange {max=-1, min=-22}}
    boneToModify {boneName=dyn_pants_out_l}
    rotation=<rot 8.682 -7.926 0.058>
   }
   n28 = "Transform (Modify) Bone - Bone: dyn_pants_b_r"!(ComponentPose=@n27, bAlphaBoolEnabled:bool=Get bIsMoving2D(), Translation="0,0,0", Rotation="0.002,-13.093,0.027", Scale="1,1,1") -> (Pose) {
    alphaInputType=Bool
    rotationMode=BMM_Additive
    rotationSpace=BCS_BoneSpace
    translationMode=BMM_Additive
    translationSpace=BCS_BoneSpace
    alphaBoolBlend {blendInTime=0.1, blendOption=Linear, blendOutTime=0.2}
    alphaScaleBiasClamp {bInterpResult=true, bMapRange=true, inRange {max=-1, min=-22}}
    boneToModify {boneName=dyn_pants_b_r}
    rotation=<rot 0.002 -13.093 0.027>
   }
   Output Pose(Result=Component To Local(ComponentPose=@n28)) {layerGroup=None}
   n29 = Get gravity_override() -> (gravity_override:Vector)
   n30 = RigidBody!(
    ComponentPose=Local To Component(LocalPose=Copy Pose From Mesh() {bUseAttachedParent=true})
    Alpha
    OverrideWorldGravity:Vector=@n29
    ExternalForce
   ) -> (Pose) {
    bClampLinearTranslationLimitToRefPose=true
    bOverrideWorldGravity=true
    overridePhysicsAsset=&/M_MED_DapperPunch_Physics
    simulationSpace=BaseBoneSpace
    alphaBoolBlend {blendOption=Linear}
    baseBoneRef {boneName=pelvis}
    componentAppliedLinearAccClamp=<3000 3000>
    componentLinearAccScale=<0.6 0.6 0.8>
    componentLinearVelScale=<0.6 0.6 0.8>
    overrideWorldGravity {z=-980}
    simSpaceSettings {maxAngularAcceleration=0, maxAngularVelocity=15000, maxLinearAcceleration=0, maxLinearVelocity=15000, worldAlpha=1}
   }
   n31 = Anim Dynamics - Bone: dyn_pants_base_l(ComponentPose=@n30, Alpha=1, ExternalForce=@n29) {
    angularDampingOverride=0.999
    angularSpringConstant=100
    bAngularSpring=true
    bGravityOverrideInSimSpace=true
    bLinearSpring=true
    bOverrideAngularDamping=true
    bOverrideLinearDamping=true
    bUseGravityOverride=true
    gravityScale=0.75
    linearDampingOverride=0.9999
    linearSpringConstant=500
    simulationSpace=RootRelative
    alphaBoolBlend {blendOption=Linear}
    alphaScaleBiasClamp {scale=0.2}
    boundBone {boneName=dyn_pants_base_l}
    componentAppliedLinearAccClamp=<6000 6000 6000>
    componentLinearAccScale=<0.25 0.25 1>
    componentLinearVelScale=<0.25 0.25 1>
    physicsBodyDefinitions [{
      sphereCollisionRadius=0
      boundBone {boneName=dyn_pants_base_l}
      boxExtents=<8 25 25>
      constraintSetup {angularLimitsMax=<0 0 0>, angularLimitsMin=<0 0 0>, angularTarget=<1 0 0>, linearAxesMax=<5 1 5>, linearAxesMin=<-5 -1 -5>}
      localJointOffset=<1 0 0>
     }]
    relativeSpaceBone {boneName=root}
   }
   n32 = Anim Dynamics - Bone: dyn_pants_b_l(ComponentPose=@n31, Alpha=1, ExternalForce=@n29) {
    angularDampingOverride=0.999
    angularSpringConstant=5000
    bAngularSpring=true
    bGravityOverrideInSimSpace=true
    bLinearSpring=true
    bOverrideAngularDamping=true
    bUseGravityOverride=true
    gravityScale=0.75
    linearSpringConstant=5000
    simulationSpace=RootRelative
    alphaBoolBlend {blendOption=Linear}
    alphaScaleBiasClamp {scale=0.2}
    boundBone {boneName=dyn_pants_b_l}
    componentAppliedLinearAccClamp=<6000 6000 6000>
    componentLinearAccScale=<0.25 0.25 0.6>
    componentLinearVelScale=<0.25 0.25 0.6>
    physicsBodyDefinitions [{
      sphereCollisionRadius=0
      boundBone {boneName=dyn_pants_b_l}
      boxExtents=<8 25 25>
      constraintSetup {angularLimitsMax=<40 40 0>, angularLimitsMin=<-40 -40 -40>, angularTarget=<1 0 0>}
      localJointOffset=<4 0 0>
     }]
    relativeSpaceBone {boneName=root}
   }
   n33 = Anim Dynamics - Bone: dyn_pants_f_l(ComponentPose=@n32, Alpha=1, ExternalForce=@n29) {
    angularDampingOverride=0.999
    angularSpringConstant=5000
    bAngularSpring=true
    bGravityOverrideInSimSpace=true
    bLinearSpring=true
    bOverrideAngularDamping=true
    bUseGravityOverride=true
    gravityScale=0.75
    linearSpringConstant=5000
    simulationSpace=RootRelative
    alphaBoolBlend {blendOption=Linear}
    alphaScaleBiasClamp {scale=0.2}
    boundBone {boneName=dyn_pants_f_l}
    componentAppliedLinearAccClamp=<6000 6000 6000>
    componentLinearAccScale=<0.25 0.25 0.6>
    componentLinearVelScale=<0.25 0.25 0.6>
    physicsBodyDefinitions [{
      sphereCollisionRadius=0
      boundBone {boneName=dyn_pants_f_l}
      boxExtents=<8 25 25>
      constraintSetup {angularLimitsMax=<60 40 -10>, angularLimitsMin=<-60 -40 -40>, angularTarget=<1 0 0>}
      localJointOffset=<4 0 0>
     }]
    relativeSpaceBone {boneName=root}
   }
   n34 = Anim Dynamics - Bone: dyn_pants_out_l(ComponentPose=@n33, Alpha=1, ExternalForce=@n29) {
    angularDampingOverride=0.999
    angularSpringConstant=5000
    bAngularSpring=true
    bGravityOverrideInSimSpace=true
    bLinearSpring=true
    bOverrideAngularDamping=true
    bUseGravityOverride=true
    gravityScale=0.75
    linearSpringConstant=5000
    simulationSpace=RootRelative
    alphaBoolBlend {blendOption=Linear}
    alphaScaleBiasClamp {scale=0.2}
    boundBone {boneName=dyn_pants_out_l}
    componentAppliedLinearAccClamp=<6000 6000 6000>
    componentLinearAccScale=<0.25 0.25 0.6>
    componentLinearVelScale=<0.25 0.25 0.6>
    physicsBodyDefinitions [{
      sphereCollisionRadius=0
      boundBone {boneName=dyn_pants_out_l}
      boxExtents=<8 25 25>
      constraintSetup {angularLimitsMax=<60 40 -20>, angularLimitsMin=<-60 -40 -70>, angularTarget=<1 0 0>}
      localJointOffset=<4 0 0>
     }]
    relativeSpaceBone {boneName=root}
   }
   n35 = Anim Dynamics - Bone: dyn_pants_inn_l(ComponentPose=@n34, Alpha=1, ExternalForce=@n29) {
    angularDampingOverride=0.999
    angularSpringConstant=5000
    bAngularSpring=true
    bGravityOverrideInSimSpace=true
    bLinearSpring=true
    bOverrideAngularDamping=true
    bUseGravityOverride=true
    gravityScale=0.75
    linearSpringConstant=5000
    simulationSpace=RootRelative
    alphaBoolBlend {blendOption=Linear}
    alphaScaleBiasClamp {scale=0.2}
    boundBone {boneName=dyn_pants_inn_l}
    componentAppliedLinearAccClamp=<6000 6000 6000>
    componentLinearAccScale=<0.25 0.25 0.6>
    componentLinearVelScale=<0.25 0.25 0.6>
    physicsBodyDefinitions [{
      sphereCollisionRadius=0
      boundBone {boneName=dyn_pants_inn_l}
      boxExtents=<8 25 25>
      constraintSetup {angularLimitsMax=<60 40 20>, angularLimitsMin=<-60 -40 -40>, angularTarget=<1 0 0>}
      localJointOffset=<4 0 0>
     }]
    relativeSpaceBone {boneName=root}
   }
   n36 = Anim Dynamics - Bone: dyn_pants_base_r(ComponentPose=@n35, Alpha=1, ExternalForce=@n29) {
    angularDampingOverride=0.999
    angularSpringConstant=100
    bAngularSpring=true
    bGravityOverrideInSimSpace=true
    bLinearSpring=true
    bOverrideAngularDamping=true
    bOverrideLinearDamping=true
    bUseGravityOverride=true
    gravityScale=0.75
    linearDampingOverride=0.9999
    linearSpringConstant=500
    simulationSpace=RootRelative
    alphaBoolBlend {blendOption=Linear}
    alphaScaleBiasClamp {scale=0.1}
    boundBone {boneName=dyn_pants_base_r}
    componentAppliedLinearAccClamp=<6000 6000 6000>
    componentLinearAccScale=<0.25 0.25 1>
    componentLinearVelScale=<0.25 0.25 1>
    physicsBodyDefinitions [{
      sphereCollisionRadius=0
      boundBone {boneName=dyn_pants_base_r}
      boxExtents=<8 25 25>
      constraintSetup {angularLimitsMax=<0 0 0>, angularLimitsMin=<0 0 0>, angularTarget=<1 0 0>, linearAxesMax=<5 5 5>, linearAxesMin=<-5 -1 -5>}
      localJointOffset=<1 0 0>
     }]
    relativeSpaceBone {boneName=root}
   }
   n37 = Anim Dynamics - Bone: dyn_pants_b_r(ComponentPose=@n36, Alpha=1, ExternalForce=@n29) {
    angularDampingOverride=0.999
    angularSpringConstant=5000
    bAngularSpring=true
    bGravityOverrideInSimSpace=true
    bLinearSpring=true
    bOverrideAngularDamping=true
    bUseGravityOverride=true
    gravityScale=0.75
    linearSpringConstant=5000
    simulationSpace=RootRelative
    alphaBoolBlend {blendOption=Linear}
    alphaScaleBiasClamp {scale=0.2}
    boundBone {boneName=dyn_pants_b_r}
    componentAppliedLinearAccClamp=<6000 6000 6000>
    componentLinearAccScale=<0.25 0.25 0.6>
    componentLinearVelScale=<0.25 0.25 0.6>
    physicsBodyDefinitions [{
      sphereCollisionRadius=0
      boundBone {boneName=dyn_pants_b_r}
      boxExtents=<8 25 25>
      constraintSetup {angularLimitsMax=<60 40 -5>, angularLimitsMin=<-60 -40 -40>, angularTarget=<1 0 0>}
      localJointOffset=<4 0 0>
     }]
    relativeSpaceBone {boneName=root}
   }
   n38 = Anim Dynamics - Bone: dyn_pants_out_r(ComponentPose=@n37, Alpha=1, ExternalForce=@n29) {
    angularDampingOverride=0.999
    angularSpringConstant=5000
    bAngularSpring=true
    bGravityOverrideInSimSpace=true
    bLinearSpring=true
    bOverrideAngularDamping=true
    bUseGravityOverride=true
    gravityScale=0.75
    linearSpringConstant=5000
    simulationSpace=RootRelative
    alphaBoolBlend {blendOption=Linear}
    alphaScaleBiasClamp {scale=0.2}
    boundBone {boneName=dyn_pants_out_r}
    componentAppliedLinearAccClamp=<6000 6000 6000>
    componentLinearAccScale=<0.25 0.25 0.6>
    componentLinearVelScale=<0.25 0.25 0.6>
    physicsBodyDefinitions [{
      sphereCollisionRadius=0
      boundBone {boneName=dyn_pants_out_r}
      boxExtents=<8 25 25>
      constraintSetup {angularLimitsMax=<60 40 -10>, angularLimitsMin=<-60 -40 -40>, angularTarget=<1 0 0>}
      localJointOffset=<4 0 0>
     }]
    relativeSpaceBone {boneName=root}
   }
   n39 = Anim Dynamics - Bone: dyn_pants_inn_r(ComponentPose=@n38, Alpha=1, ExternalForce=@n29) {
    angularDampingOverride=0.999
    angularSpringConstant=5000
    bAngularSpring=true
    bGravityOverrideInSimSpace=true
    bLinearSpring=true
    bOverrideAngularDamping=true
    bUseGravityOverride=true
    gravityScale=0.75
    linearSpringConstant=5000
    simulationSpace=RootRelative
    alphaBoolBlend {blendOption=Linear}
    alphaScaleBiasClamp {scale=0.3}
    boundBone {boneName=dyn_pants_inn_r}
    componentAppliedLinearAccClamp=<6000 6000 6000>
    componentLinearAccScale=<0.25 0.25 0.6>
    componentLinearVelScale=<0.25 0.25 0.6>
    physicsBodyDefinitions [{
      sphereCollisionRadius=1
      boundBone {boneName=dyn_pants_inn_r}
      boxExtents=<45 10 25>
      constraintSetup {angularLimitsMax=<60 20 -30>, angularLimitsMin=<20 -20 -60>, angularTarget=<1 0 0>}
      localJointOffset=<-4 0 0>
     }]
    relativeSpaceBone {boneName=root}
    sphericalLimits [{limitRadius=5, drivingBone {boneName=foot_r}, sphereLocalOffset=<0 0 5>}]
   }
   Save cached pose 'SavedPose28735'(Pose=Component To Local(ComponentPose=@n39)) {cacheName=SavedPose28735}
  ]}
 EventGraph {nodes [
   Event BlueprintUpdateAnimation()
   > n1 = Sequence!(execute) -> (then_0, then_1, then_2:Exec, then_3:Exec, then_4:Exec)
   n2 = Gravity Override State Function DEPRECATED(
    execute=@n1.then_1
    CharacterPartAnimInstance=Self-Reference()
    GameTimeInSeconds=GetGameTimeInSeconds()
    GravityOverrideParams=Get Gravity Override Params()
   )
   > Set gravity_override(gravity_override=@n2)
   n3 = GetOwnerSkeletalMeshComponent()
   n4 = GetDeltaTransformFromRefPose!(self=@n3, BoneName=upperarm_r, BaseName=spine_05) -> (
    ReturnValue_Location:Vector
    ReturnValue_Rotation_Roll:float
    ReturnValue_Rotation_Pitch:float
    ReturnValue_Rotation_Yaw:float
    ReturnValue_Scale:Vector
   )
   Set r_upperarm_angle_for_b_bag(
    execute=@n1.then_2
    r_upperarm_angle_for_b_bag:double=float * float(
     A=MapRangeClamped(Value=@n4.ReturnValue_Rotation_Pitch, InRangeA=-10, InRangeB=-20)
     B=MapRangeClamped(Value=@n4.ReturnValue_Rotation_Yaw, InRangeA=50, InRangeB=30)
    ) {is=title K2Node_PromotableOperator}
   ) -> (Output_Get:double)
   > Set r_upperarm_angle_for_f_bag(r_upperarm_angle_for_f_bag:double=float * float(
     A=MapRangeClamped(Value=@n4.ReturnValue_Rotation_Pitch, InRangeA=-10, InRangeB=-20)
     B=MapRangeClamped(Value=@n4.ReturnValue_Rotation_Yaw, InRangeA=50, InRangeB=35)
    )) -> (Output_Get:double)
   n5 = GetDeltaTransformFromRefPose(self=@n3, BoneName=upperarm_l, BaseName=spine_05)
   > Set l_upperarm_angle_for_b_bag(l_upperarm_angle_for_b_bag:double=float * float(
     A=MapRangeClamped(Value=@n5.ReturnValue_Rotation_Pitch, InRangeA=-10, InRangeB=-20)
     B=MapRangeClamped(Value=@n5.ReturnValue_Rotation_Yaw, InRangeA=35, InRangeB=25)
    )) -> (Output_Get:double)
   > Set l_upperarm_angle_for_f_bag(l_upperarm_angle_for_f_bag:double=float * float(
     A=MapRangeClamped(Value=@n5.ReturnValue_Rotation_Pitch, InRangeA=-10, InRangeB=-20)
     B=MapRangeClamped(Value=@n5.ReturnValue_Rotation_Yaw, InRangeA=50, InRangeB=35)
    )) -> (Output_Get:double)
   > Set l_foot_angle_alpha(l_foot_angle_alpha:double=MapRangeClamped(Value=GetDeltaTransformFromRefPose(self=@n3, BoneName=foot_l, BaseName=calf_l).ReturnValue_Rotation_Yaw, InRangeA=0, InRangeB=-26)) -> (Output_Get:double)
   n6 = Get bIsCrouching()
   Set CrouchingAlpha(
    execute=@n1.then_3
    CrouchingAlpha:double=FInterpTo(
     Current=Get CrouchingAlpha()
     Target="To Float (Boolean)"(InBool=@n6)
     DeltaTime=GetWorldDeltaSeconds!() -> (ReturnValue:double)
     InterpSpeed=25
    )
   ) -> (Output_Get:double)
   n7 = Get bIsSliding()
   > Set Sliding(Sliding:double=FInterpTo(Current=Get Sliding(), Target="To Float (Boolean)"(InBool=@n7), DeltaTime=GetWorldDeltaSeconds(), InterpSpeed=20)) -> (Output_Get:double)
   n8 = Get bIsMoving2D()
   > Set moving2d(moving2d:double=FInterpTo(Current=Get moving2d(), Target="To Float (Boolean)"(InBool=@n8), DeltaTime=GetWorldDeltaSeconds(), InterpSpeed=10)) -> (Output_Get:double)
  ]}
]
variables [
 CrouchingAlpha {type=number}
 Gravity Override Params {
  type=GravityOverrideParamsStruct
  default {
   skydive_FloatingDown {gravityOverride=<0 0 -100>, windAmplitude=<0 0 0>, windFrequency=<0 0 0>, windOffset=<0 0 0>}
   skydive_GoingUp {gravityOverride=<0 0 -100>, windAmplitude=<0 0 0>, windFrequency=<0 0 0>}
  }
 }
 gravity_override {type=Vector}
 l_foot_angle_alpha {type=number}
 l_upperarm_angle_for_b_bag {type=number}
 l_upperarm_angle_for_f_bag {type=number}
 moving2d {type=number}
 r_upperarm_angle_for_b_bag {type=number}
 r_upperarm_angle_for_f_bag {type=number}
 Sliding {type=number}
]
